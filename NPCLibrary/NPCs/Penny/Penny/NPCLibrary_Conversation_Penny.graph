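# Conversation database for the Penny NPC, defining the greeting and options players get when talking to Penny.
# The entry tag leads to one speech node with a default line and contextual variants for time since last match, first meeting, Bars held and specific outfits.
# Each variant's priority comes from a Tandem speech curve row.
# The speech then branches through link nodes into the shared NPCLibrary Quests and BasicServices conversations.
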ConversationDatabase /NPCLibrary/NPCs/Penny/Penny/NPCLibrary_Conversation_Penny
entry_points [{tag=Conversation.Entry.Penny, destinations [#1]}]
nodes [
 ConversationEntryPointNode_7 {
  id=#1
  class=ConversationEntryPointNode
  outputs [#2]
  properties {entryTag {tagName=Conversation.Entry.Penny}}
 }
 Speech_C_4 {
  id=#2
  class=Speech_C
  outputs [#3, #4]
  properties {
   nodeName=""
   generalConfig {
    defaultMessage="Have your buildings been up to code, then? "
    contextualMessages [
     {
      message="We can always use another pair of hands, especially in these dreadful times."
      contextRequirements [{requirement {class=FortControllerRequirement_CheckTimeSinceLastPlayedMatch}}]
      priority {curve {rowName=Tandem.Speech.Priority.WhileSincePlayed}}
     }
     {
      message=What have you been up to? Making trouble?
      contextRequirements [{requirement {class=FortControllerRequirement_IsFirstEverConversationWithNPC}}]
      priority {curve {rowName=Tandem.Speech.Priority.FirstEverInteraction}}
     }
     {
      message="More Bars, stronger material."
      contextRequirements [{requirement {class=FortControllerRequirement_HasBars}}]
      priority {curve {rowName=Tandem.Speech.Priority.HighTotalBars}}
     }
     {
      message="You've joined me in cleaning up this mess, have you? "
      contextRequirements [{requirement {
         class=FortControllerRequirement_HasCID
         properties {allowedCIds [
           /BRCosmetics/Athena/Items/Cosmetics/Characters/CID_719_Athena_Commando_F_Blonde
           /BRCosmetics/Athena/Items/Cosmetics/Characters/CID_A_186_Athena_Commando_M_Lars
           /BRCosmetics/Athena/Items/Cosmetics/Characters/CID_203_Athena_Commando_M_CloakedStar
          ]}
        }}]
      priority {curve {rowName=Tandem.Speech.Priority.OutfitReaction}}
     }
    ]
   }
  }
 }
 ConversationLinkNode_0 {id=#4, class=ConversationLinkNode, properties {remoteEntryTag {tagName=Conversation.Entry.BasicServices}}}
 ConversationLinkNode_1 {id=#3, class=ConversationLinkNode, properties {remoteEntryTag {tagName=Conversation.Entry.NPCLibrary.Quests}}}
]
